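; Spawned actor that spells FORTNITE from completed letter quests, celebrating when all are done.
; On BeginPlay it plays a sound, creates dynamic materials and unhides letters whose quests are complete.
; The DropLetters timeline lowers the letters on shuffled tracks, then starts FadeOutLetters and, if complete, FinalFadeOut's delayed sound and particles.
; FadeOutLetters scales and fades only the F, then destroys the actor.
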(event EventBeginPlay
  (bind _letters_done (Variables|Default|GetLetters_Done))
  (CallFunction|SetUpDynamicMats)
  (Audio|PlaySoundatLocation (Variables|Default|GetSpellingAppearSound) (Transformation|GetWorldLocation (Variables|Default|GetDefaultSceneRoot)))
  (Variables|Default|SetFinalLetter (== (Utilities|Array|Length (Variables|Default|GetAll_Letter_Quests)) (Utilities|Array|Length (Variables|Default|GetCompleted_Letter_Quests))))
  (for (_array_element _array_index) (Variables|Default|GetAll_Letter_Quests)
    (Utilities|Map|Add (Variables|Default|GetLetters_Done) _array_index (Utilities|Array|ContainsItem (Variables|Default|GetCompleted_Letter_Quests) _array_element)))
  (bind _keys (Utilities|Map|Keys _letters_done))
  (for (_ _array_index_2) _keys
    (if (Utilities|Map|Find _letters_done _array_index_2)
      (Development|PrintString "True" false false "(R=0.240457,G=1.000000,B=0.000000,A=1.000000)")
      (switch int _array_index_2
        (:0
          (Development|SetHiddeninGame (Variables|Default|GetSM_Fortnite_F) false false))
        (:1
          (Development|SetHiddeninGame (Variables|Default|GetSM_Fortnite_O) false false))
        (:2
          (Development|SetHiddeninGame (Variables|Default|GetSM_Fortnite_R) false false))
        (:3
          (Development|SetHiddeninGame (Variables|Default|GetSM_Fortnite_T1) false false))
        (:4
          (Development|SetHiddeninGame (Variables|Default|GetSM_Fortnite_N) false false))
        (:5
          (Development|SetHiddeninGame (Variables|Default|GetSM_Fortnite_I) false false))
        (:6
          (Development|SetHiddeninGame (Variables|Default|GetSM_Fortnite_T2) false false))
        (:7
          (Development|SetHiddeninGame (Variables|Default|GetSM_Fortnite_E) false false)))))
  (Utilities|FlowControl|Sequence
    (:then_1
      (bind (_direction _f_time _o_time _r_time _t1_time _n_time _i_time _t2_time _e_time) (Timeline|DropLetters PlayFromStart)
        (:Update
          (Transformation|SetRelativeLocation (Variables|Default|GetSM_Fortnite_F) (Math|Vector|Lerp(Vector) "0.000000,155.000000,500.000000" "0.000000,155.000000, 0" _o_time) false false)
          (Transformation|SetRelativeLocation (Variables|Default|GetSM_Fortnite_O) (Math|Vector|Lerp(Vector) "0.000000,110.000000,500.000000" "0,110.000000, 0" _t2_time) false false)
          (Transformation|SetRelativeLocation (Variables|Default|GetSM_Fortnite_R) (Math|Vector|Lerp(Vector) "0.000000,50.000000,500.000000" "0,50.000000, 0" _e_time) false false)
          (Transformation|SetRelativeLocation (Variables|Default|GetSM_Fortnite_T1) (Math|Vector|Lerp(Vector) "0.000000,0.000000,500.000000" "0, 0, 0" _t1_time) false false)
          (Transformation|SetRelativeLocation (Variables|Default|GetSM_Fortnite_N) (Math|Vector|Lerp(Vector) "0.000000,-55.000000,500.000000" "0,-55.000000, 0" _f_time) false false)
          (Transformation|SetRelativeLocation (Variables|Default|GetSM_Fortnite_I) (Math|Vector|Lerp(Vector) "0.000000,-100.000000,500.000000" "0,-100.000000, 0" _i_time) false false)
          (Transformation|SetRelativeLocation (Variables|Default|GetSM_Fortnite_T2) (Math|Vector|Lerp(Vector) "0.000000,-140.000000,500.000000" "0,-140.000000, 0" _n_time) false false)
          (Transformation|SetRelativeLocation (Variables|Default|GetSM_Fortnite_E) (Math|Vector|Lerp(Vector) "0.000000,-190.000000,500.000000" "0,-190.000000, 0" _r_time) false false))
        (:Finished
          (CallFunction|FadeOutLetters)
          (if (Variables|Default|GetFinalLetter)
            (CallFunction|FinalFadeOut)))))))

(event Custom|FadeOutLetters
  (Utilities|FlowControl|Delay 0.2)
  (bind (_direction _fadeouttime) (Timeline|FadeOut PlayFromStart)
    (:Update
      (Transformation|SetRelativeScale3D (Variables|Default|GetSM_Fortnite_F) (Math|Vector|Lerp(Vector) "1.000000,1.000000,1.000000" "1.250000,1.250000,1.250000" _fadeouttime))
      (Rendering|Material|SetScalarParameterValue (Variables|Default|GetF_Mat) "Opacity" (Math|Float|Lerp 1.0 0.0 _fadeouttime) :!class /Script/Engine.MaterialInstanceDynamic))
    (:Finished
      (Actor|DestroyActor))))

(event Custom|FinalFadeOut
  (Utilities|FlowControl|Delay 0.45)
  (Audio|PlaySoundatLocation (Variables|Default|GetCelebrationSound) (Transformation|GetWorldLocation (Variables|Default|GetDefaultSceneRoot)))
  (Components|Activation|Activate (Variables|Default|GetP_Celebrate1) false)
  (Utilities|FlowControl|Delay 0.15)
  (Components|Activation|Activate (Variables|Default|GetP_Celebrate) false)
  (Utilities|FlowControl|Delay 0.1)
  (Components|Activation|Activate (Variables|Default|GetP_Celebrate2) false))

(fn SetUpDynamicMats
  (bind _returnvalue (Rendering|Material|CreateDynamicMaterialInstance (Variables|Default|GetSM_Fortnite_F) 0 "/Game/Athena/Items/QuestInteractables/FortnightLetters/FX/M_FortniteLetterMat.M_FortniteLetterMat" "None" :!class /Script/Engine.PrimitiveComponent))
  (Variables|Default|SetF_Mat _returnvalue)
  (bind _returnvalue_2 (Rendering|Material|CreateDynamicMaterialInstance (Variables|Default|GetSM_Fortnite_O) 0 "/Game/Athena/Items/QuestInteractables/FortnightLetters/FX/M_FortniteLetterMat.M_FortniteLetterMat" "None" :!class /Script/Engine.PrimitiveComponent))
  (Variables|Default|SetO_Mat _returnvalue_2)
  (bind _returnvalue_3 (Rendering|Material|CreateDynamicMaterialInstance (Variables|Default|GetSM_Fortnite_R) 0 "/Game/Athena/Items/QuestInteractables/FortnightLetters/FX/M_FortniteLetterMat.M_FortniteLetterMat" "None" :!class /Script/Engine.PrimitiveComponent))
  (Variables|Default|SetR_Mat _returnvalue_3)
  (bind _returnvalue_4 (Rendering|Material|CreateDynamicMaterialInstance (Variables|Default|GetSM_Fortnite_T1) 0 "/Game/Athena/Items/QuestInteractables/FortnightLetters/FX/M_FortniteLetterMat.M_FortniteLetterMat" "None" :!class /Script/Engine.PrimitiveComponent))
  (Variables|Default|SetT1_Mat _returnvalue_4)
  (bind _returnvalue_5 (Rendering|Material|CreateDynamicMaterialInstance (Variables|Default|GetSM_Fortnite_N) 0 "/Game/Athena/Items/QuestInteractables/FortnightLetters/FX/M_FortniteLetterMat.M_FortniteLetterMat" "None" :!class /Script/Engine.PrimitiveComponent))
  (Variables|Default|SetN_Mat _returnvalue_5)
  (bind _returnvalue_6 (Rendering|Material|CreateDynamicMaterialInstance (Variables|Default|GetSM_Fortnite_I) 0 "/Game/Athena/Items/QuestInteractables/FortnightLetters/FX/M_FortniteLetterMat.M_FortniteLetterMat" "None" :!class /Script/Engine.PrimitiveComponent))
  (Variables|Default|SetI_Mat _returnvalue_6)
  (bind _returnvalue_7 (Rendering|Material|CreateDynamicMaterialInstance (Variables|Default|GetSM_Fortnite_T2) 0 "/Game/Athena/Items/QuestInteractables/FortnightLetters/FX/M_FortniteLetterMat.M_FortniteLetterMat" "None" :!class /Script/Engine.PrimitiveComponent))
  (Variables|Default|SetT2_Mat _returnvalue_7)
  (bind _returnvalue_8 (Rendering|Material|CreateDynamicMaterialInstance (Variables|Default|GetSM_Fortnite_E) 0 "/Game/Athena/Items/QuestInteractables/FortnightLetters/FX/M_FortniteLetterMat.M_FortniteLetterMat" "None" :!class /Script/Engine.PrimitiveComponent))
  (Variables|Default|SetE_Mat _returnvalue_8))

(fn ConstructionScript)
